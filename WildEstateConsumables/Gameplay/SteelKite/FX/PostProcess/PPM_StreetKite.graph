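# Post-process material for the SteelKite consumable (StreetKite) that tints and distorts the screen through centered masks.
# A sine-distorted, panning distance gradient drives HueShift for a contrast-boosted rainbow, lerped with a main color and offset scene color.
# Sphere masks, modulated by panning fractal noise and a star flare, blend this with the untouched scene.
# Named reroutes share the masks, noise and a procedural radial ripple band.
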Material /WildEstateConsumables/Gameplay/SteelKite/FX/PostProcess/PPM_StreetKite
graph {
 properties {bAutomaticallySetUsageInEditor=false, materialDomain=MD_PostProcess, usedShadingModels=Default Lit}
 nodes [
  Constant3Vector()
  n1 = TextureCoordinate()
  n2 = VectorParameter() {group=Color, parameterName=UvMultAndPanSpeed, defaultValue=<rgba 32 1 12 0.5>}
  n3 = AppendVector(A=@n2.RGB, B=@n2.RGB)
  n4 = ComponentMask(None=Panner(Coordinate=TextureCoordinate() {uTiling=1.0242, vTiling=8}, Speed=ComponentMask(None=@n3) {a=true, b=true}) {speedY=0.05}) {g=true}
  n5 = Add(A=Multiply(A=TextureCoordinate(), B=ComponentMask(None=@n3) {g=true, r=true}), B=Multiply(A=Sine(None=@n4)) {constB=0.5})
  n6 = Multiply(
   A=Sine(None=ComponentMask(None=@n5) {g=true, r=true}) {period=2}
   B=ScalarParameter() {defaultValue=0.1, group=Color, parameterName=DistortionAmount_Color}
  )
  n7 = Multiply(A=Distance(A=Add(A=@n1, B=@n6), B=Constant() {r=0.5}), B=ScalarParameter() {defaultValue=5, group=Color, parameterName=Color_Tile})
  n8 = Time()
  n9 = ScalarParameter() {group=Color, parameterName=Color_Panner_speed}
  n10 = OneMinus(None=Frac(None=Multiply(A=@n8, B=@n9)))
  n11 = Multiply(
   A=MaterialFunctionCall(
    Hue Shift Percentage=Add(A=@n7, B=@n10)
    Texture=VectorParameter() {group=Color, parameterName=HueShift_Tone, defaultValue=<rgba 1 0 0 1>}.RGB
   ) {materialFunction=&HueShift}.Result
   B=ScalarParameter() {defaultValue=1, group=Color, parameterName=color_intensity}
  )
  n12 = ScalarParameter() {defaultValue=100, group=Color, parameterName=Cheap_Contrast}
  n13 = NamedRerouteUsage()
  n14 = NamedRerouteUsage()
  n15 = LinearInterpolate(
   A=Multiply(A=MaterialFunctionCall!(In=@n11, Contrast=@n12) {materialFunction=/Engine/Functions/Engine_MaterialFunctions01/ImageAdjustment/CheapContrast_RGB}.Result)
   B=VectorParameter() {group=Color, parameterName=Main_Color, defaultValue=<rgba 1 0.0931 0 1>}.RGB
   Alpha=Saturate(None=Multiply(A=@n13, B=@n14))
  ) {constAlpha=0}
  n16 = TextureCoordinate()
  n17 = Multiply(A=NamedRerouteUsage(), B=ScalarParameter() {defaultValue=0.1, group=Ripple, parameterName=DistortionAmount_Screen})
  n18 = ComponentMask(None=SceneTexture(UVs=Add(A=@n16, B=@n17)) {sceneTextureId=PPI_PostProcessInput0}.Color) {b=true, g=true, r=true}
  n19 = LinearInterpolate(
   A=ScalarParameter() {group=Controls, parameterName=blend_color_rainbow}
   B=ScalarParameter() {group=Controls, parameterName=blend_color_main}
   Alpha=Multiply(A=NamedRerouteUsage(), B=NamedRerouteUsage())
  ) {constAlpha=0}
  n20 = Constant2Vector() {g=0.5, r=0.5}
  n21 = ScreenPosition()
  n22 = MaterialFunctionCall(Float2=@n21.ViewportUV) {materialFunction=&BreakOutFloat2Components}
  n23 = MaterialFunctionCall(Float2=MaterialFunctionCall() {materialFunction=&ScreenResolution}.Buffer Resolution) {materialFunction=&BreakOutFloat2Components}
  n24 = Divide(A=@n23.R, B=@n23.R)
  n25 = LinearInterpolate(
   A=@n21.ViewportUV
   B=AppendVector(A=@n22.R, B=Multiply(A=@n24, B=@n22.G))
   Alpha=ScalarParameter() {group=Controls, parameterName=UseScaledUVToCalcMask}
  )
  n26 = ScalarParameter() {defaultValue=0.5, group=Noise, parameterName=CenterMaskSize}
  n27 = LinearInterpolate(
   A=ScalarParameter() {group=Noise, parameterName=CenterMaskhardness}
   B=ScalarParameter() {group=Noise, parameterName=CenterMaskBhardness}
   Alpha=NamedRerouteUsage()
  )
  LinearInterpolate(
   A=ComponentMask(None=LinearInterpolate(A=@n15, B=@n18, Alpha=@n19)) {b=true, g=true, r=true}
   B=ComponentMask(None=SceneTexture() {sceneTextureId=PPI_PostProcessInput0}.Color) {b=true, g=true, r=true}
   Alpha=LinearInterpolate(B=SphereMask(A=@n20, B=@n25, Radius=@n26, Hardness=@n27), Alpha=NamedRerouteUsage()) {constA=1, constB=0}
  ) {constAlpha=0}
  n28 = ScalarParameter() {group=Noise, parameterName=Noise_SpeedX}
  n29 = ScalarParameter() {group=Noise, parameterName=Noise_SpeedY}
  n30 = TextureSample(UVs=Panner(Coordinate=ViewProperty() {property=MEVP_ResolutionFraction}.Property, Speed=AppendVector(A=@n28, B=@n29))) {automaticViewMipBias=false, texture=/Game/Global/Textures/Default/Caustics/T_Fractal_01}
  n31 = ScalarParameter() {defaultValue=0.344, group=Noise, parameterName=noise_Str_overall}
  n32 = TextureSample(UVs=MaterialFunctionCall(Texture Scale=ScalarParameter() {defaultValue=1.4385, group=Noise, parameterName=Radial_Star_Scale}) {materialFunction=&ScaleUVsByCenter}.UVs) {samplerType=SAMPLERTYPE_LinearColor, texture=/Game/Effects/Fort_Effects/GameplayPulginsAssets/DiamondSkin/Textures/T_DiamondSkin_Flare}
  n33 = Saturate(None=@n32.R)
  n34 = ScalarParameter() {group=Noise, parameterName=star_str}
  n35 = Saturate(None=Add(A=Multiply(A=@n30.R, B=@n31), B=Multiply(A=@n33, B=@n34)))
  LinearInterpolate(
   A=@n26
   Alpha=Add(
    A=NamedRerouteDeclaration(None=@n35) {name=noise, nodeColor=<rgba 0 1 0.1562 1>}
    B=ScalarParameter() {defaultValue=0.344, group=Noise, parameterName=Noise_Str}
   )
  ) {constB=0}
  MaterialFunctionCall!(UV Coord, LenghtScale, RadialScale) {materialFunction=/Game/Global/MaterialFunctions/TechArt/MF_PolarCoordinates}
  n36 = ScalarParameter() {group=Controls, parameterName=Mask_Lerp}
  NamedRerouteDeclaration(None=SphereMask(A=@n20, B=@n25, Radius=LinearInterpolate(Alpha=@n36), Hardness=ScalarParameter() {group=Controls, parameterName=Mask_Sharpness})) {name=Controls, nodeColor=<rgba 1 0 0.2109 1>}
  NamedRerouteDeclaration(None=SphereMask(
    A=Constant2Vector() {g=0.5, r=0.5}
    B=ScreenPosition().ViewportUV
    Radius=ScalarParameter() {defaultValue=0.5, group=ColorMask, parameterName=Color_Mask_Size}
    Hardness=ScalarParameter() {group=ColorMask, parameterName=Color_Mask_Hardness}
   )) {name=Color_Mask_V2, nodeColor=<rgba 0.7344 1 0 1>}
  n37 = Constant() {r=-0.5}
  n38 = MaterialFunctionCall(Float2=Add(A=AppendVector(A=@n37, B=@n37), B=TextureCoordinate())) {materialFunction=&BreakOutFloat2Components}
  n39 = ScalarParameter() {defaultValue=8, group=Ripple, parameterName=Radial_Gradient_Pow}
  n40 = Multiply(A=@n38.R, B=@n39)
  n41 = Multiply(A=@n39, B=@n38.G)
  n42 = Panner(
   Coordinate=Add(A=Multiply(A=@n40, B=@n40), B=Multiply(A=@n41, B=@n41))
   Speed=ScalarParameter() {defaultValue=-1, group=Ripple, parameterName=ripple_speed}
  )
  NamedRerouteDeclaration(None=Multiply(
    A=MaterialFunctionCall!(Width, Sharpness, Offset, Direction Switch, Compare, Input Coordinates=@n42) {materialFunction=/Engine/Functions/Engine_MaterialFunctions03/Procedurals/GeneratedBand}.Result
    B=ScalarParameter() {defaultValue=0.184, group=Ripple, parameterName=Radial_Gradient_Str}
   )) {name=Radial_Warp, nodeColor=<rgba 0 1 0.625 1>}
  n43 = ScalarParameter() {group=Controls, parameterName=Interior_Radius_Max}
  n44 = ScalarParameter() {group=Controls, parameterName=Interior_Radius}
  n45 = NamedRerouteUsage()
  NamedRerouteDeclaration(None=SphereMask(
    A=Constant2Vector() {g=0.5, r=0.5}
    B=@n25
    Radius=LinearInterpolate(A=@n43, B=@n44, Alpha=@n45)
    Hardness=ScalarParameter() {group=Controls, parameterName=interior_Sharpness}
   )) {name=interior_radius, nodeColor=<rgba 0 1 0.5547 1>}
  NamedRerouteDeclaration(None=@n7) {name=distortion noise, nodeColor=<rgba 0 1 1 1>}
  ScreenPosition()
  TextureCoordinate()
  VectorParameter() {group=Print, parameterName=Print Tint, defaultValue=<rgba 1 0.989 0.9688 1>}
 ]
}
